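# Sound cue that plays intro-plus-loop music for the biplane timed trial quest in Figment_S07 Athena.
# A mixer starts a one-shot intro wave alongside a fixed delay, which then starts a looping music wave to follow the intro.
# It routes through the Fort_Gadgets sound class with a slight volume boost.
# Its concurrency override uses a prevent-new rule, so new instances over the limit are rejected.
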SoundCue /Figment_S07/Athena/Sounds/Quests/TimedTrial/Athena_Quest_TimeTrial_Biplane_Loop_Cue
root=Mixer_0
nodes [
 {id=Delay_0, children [WavePlayer_2], properties {delayMax=6.272, delayMin=6.272}}
 {id=Mixer_0, children [WavePlayer_0,Delay_0], properties {inputVolume [1,1]}}
 {id=WavePlayer_0, properties {soundWaveAssetPtr=&/TimedTrial_Biplane_Music_Intro}}
 {id=WavePlayer_2, properties {bLooping=true, soundWaveAssetPtr=&/TimedTrial_Biplane_Music_Loop}}
]
properties {
 bOverrideConcurrency=true
 maxDistance=2097152
 soundClassObject=&Fort_Gadgets
 volumeMultiplier=1.15
 attenuationOverrides {falloffDistance=1000, attenuationShapeExtents {x=200}}
 concurrencyOverrides {resolutionRule=PreventNew}
}
